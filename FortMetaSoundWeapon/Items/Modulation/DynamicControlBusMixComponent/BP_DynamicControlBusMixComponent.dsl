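; Actor component that activates a temporary audio modulation Control Bus Mix scaled by distance from an effect.
; On BeginPlay it builds stages from a DynamicControlBusMixData object, interpolating each bus's value, attack and release between close and far settings across inner radius plus falloff.
; A timer deactivates the mix and destroys the component after the activation duration; EndPlay also deactivates it.
; Dispatchers announce activation and deactivation, alongside debug prints.
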(event EventBeginPlay
  (for _array_element (Variables|/FortMetaSoundWeapon/Items/Modulation/DynamicControlBusMixComponent/DynamicControlBusMixData.DynamicControlBusMixData_C|GetMix%20Stages (Variables|Default|GetDynamic%20Control%20Bus%20Mix%20Data))
    (Utilities|Array|Add (Variables|Default|GetStages) (CallFunction|CreateControlBusMixStage _array_element)))
  (bind _returnvalue (Audio|Modulation|CreateControlBusMix (Variables|/FortMetaSoundWeapon/Items/Modulation/DynamicControlBusMixComponent/DynamicControlBusMixData.DynamicControlBusMixData_C|GetControl%20Bus%20Mix%20Name (Variables|Default|GetDynamic%20Control%20Bus%20Mix%20Data)) (Variables|Default|GetStages) true))
  (Variables|Default|SetControl%20Bus%20Mix _returnvalue)
  (Utilities|Time|SetTimerbyEvent (EventDispatchers|CreateEvent :!event Custom|Activation%20Duration%20Finished) (Variables|/FortMetaSoundWeapon/Items/Modulation/DynamicControlBusMixComponent/DynamicControlBusMixData.DynamicControlBusMixData_C|GetActivation%20Duration (Variables|Default|GetDynamic%20Control%20Bus%20Mix%20Data)))
  (Development|PrintText (Utilities|Text|FormatText "NSLOCTEXT(\"[B3205B942F6FDA5E731D358C7861598F]\", \"78B4541C4F703979CBDD739606F65122\", \"Activated Control Bus Mix: {cbm}\")" (Utilities|GetDisplayName (Variables|Default|GetControl%20Bus%20Mix))) false)
  (Default|CallOnControlBusMixActivated (Variables|Default|GetControl%20Bus%20Mix)))

(event Custom|Activation%20Duration%20Finished
  (CallFunction|DeactivateControlBusMix)
  (Components|DestroyComponent))

(event Custom|Deactivate%20Control%20Bus%20Mix
  (bind _control_bus_mix (Variables|Default|GetControl%20Bus%20Mix))
  (Audio|Modulation|DeactivateControlBusMix _control_bus_mix)
  (Development|PrintText (Utilities|Text|FormatText "NSLOCTEXT(\"[B3205B942F6FDA5E731D358C7861598F]\", \"D6915EA34C76F46E140515BABD080D6B\", \"Deactivated Control Bus Mix: {cbm}\")" (Utilities|GetDisplayName _control_bus_mix)) false)
  (Default|CallOnControlBusMixDeactivated (Variables|Default|GetControl%20Bus%20Mix)))

(event EventEndPlay
  (params (EndPlayReason :type EEndPlayReason%20Enum :default "Destroyed"))
  (CallFunction|DeactivateControlBusMix))

(fn CreateControlBusMixStage
  (params (DynamicControlBusMixStageSettings :type Dynamic%20Control%20Bus%20Mix%20Stage%20Settings%20Structure%20%28by%20ref%29))
  (bind (_controlbus_5_39550ecf43d9a389c1a2378c2b8671c2 _mixvalueclose_33_14631aaa441a1c35e3f89cb991dd7fcb _mixvaluefar_28_066a304f475800a203b1f79041771785 _attacktimeclose_34_9d519c8c4c8759e0646a69a6c1a73517 _attacktimefar_30_4ef68e794e563d291dcead99a7973545 _releasetimeclose_35_50e9549c47da0367ae6135b17e45eb6f _releasetimefar_32_fd4ebc9e45f1c2c61a8fb88498683f8a) (Utilities|Struct|BreakDynamicControlBusMixStageSettings DynamicControlBusMixStageSettings))
  (bind _dynamic_control_bus_mix_data (Variables|Default|GetDynamic%20Control%20Bus%20Mix%20Data))
  (bind _falloff_distance (Variables|/FortMetaSoundWeapon/Items/Modulation/DynamicControlBusMixComponent/DynamicControlBusMixData.DynamicControlBusMixData_C|GetFalloff%20Distance _dynamic_control_bus_mix_data))
  (bind _inner_radius (Variables|/FortMetaSoundWeapon/Items/Modulation/DynamicControlBusMixComponent/DynamicControlBusMixData.DynamicControlBusMixData_C|GetInner%20Radius _dynamic_control_bus_mix_data))
  (bind _returnvalue (+ _inner_radius _falloff_distance))
  (bind _distance_from_effect (Variables|Runtime|GetDistance%20from%20Effect))
  (bind _returnvalue_2 (Math|Float|MapRangeClamped _distance_from_effect _inner_radius _returnvalue _releasetimeclose_35_50e9549c47da0367ae6135b17e45eb6f _releasetimefar_32_fd4ebc9e45f1c2c61a8fb88498683f8a))
  (bind _returnvalue_3 (+ _inner_radius _falloff_distance))
  (bind _returnvalue_4 (Math|Float|MapRangeClamped _distance_from_effect _inner_radius _returnvalue_3 _mixvalueclose_33_14631aaa441a1c35e3f89cb991dd7fcb _mixvaluefar_28_066a304f475800a203b1f79041771785))
  (bind _returnvalue_5 (+ _inner_radius _falloff_distance))
  (bind _returnvalue_6 (Math|Float|MapRangeClamped _distance_from_effect _inner_radius _returnvalue_5 _attacktimeclose_34_9d519c8c4c8759e0646a69a6c1a73517 _attacktimefar_30_4ef68e794e563d291dcead99a7973545))
  (bind _returnvalue_7 (Audio|Modulation|CreateControlBusMixStage _controlbus_5_39550ecf43d9a389c1a2378c2b8671c2 _returnvalue_4 _returnvalue_6 _returnvalue_2))
  (Development|PrintText (Utilities|Text|FormatText "NSLOCTEXT(\"[B3205B942F6FDA5E731D358C7861598F]\", \"7721A0CE449ADA36DF2FE4ADF6999788\", \"Distance: {distance}, Control Bus: {cb}, Value: {value}, Attack: {atk}, Release: {rel}\")" _distance_from_effect _controlbus_5_39550ecf43d9a389c1a2378c2b8671c2 _returnvalue_4 _returnvalue_6 _returnvalue_2) false)
  (return _returnvalue_7))
